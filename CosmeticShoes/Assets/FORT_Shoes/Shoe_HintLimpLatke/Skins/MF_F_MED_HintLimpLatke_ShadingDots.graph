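# Material function for the HintLimpLatke shoe skin that adds stylized halftone shading dots and faint noise to the emissive of incoming material attributes.
# Tiled UV dots, sized by pre-skinned position, are tinted and masked by a falloff, then blended in.
# LUT-processed cloud noise is added before saturation.
# An actor-anchored, camera-facing virtual plane supplies FOV-compensated screen coordinates, alongside directional-light Lambert and camera-distance terms.
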MaterialFunction /CosmeticShoes/Assets/FORT_Shoes/Shoe_HintLimpLatke/Skins/MF_F_MED_HintLimpLatke_ShadingDots
graph {nodes [
  FunctionOutput(None=NamedRerouteUsage())
  n1 = GetMaterialAttributes(None=NamedRerouteUsage()) {attributeGetTypes [#1]}
  n2 = Add(A=@n1.EmissiveColor, B=NamedRerouteUsage())
  n3 = NamedRerouteUsage()
  n4 = NamedRerouteUsage()
  n5 = GetMaterialAttributes(None=BlendMaterialAttributes(
    A=@n1.MaterialAttributes
    B=SetMaterialAttributes!(MaterialAttributes=@n1.MaterialAttributes, EmissiveColor=@n2) {attributeSetTypes [#1]}
    Alpha=Multiply(A=@n3, B=@n4)
   )) {attributeGetTypes [#1]}
  n6 = TextureSample(UVs=Multiply(A=NamedRerouteUsage()) {constB=0.5}) {samplerSource=SSM_Wrap_WorldGroupSettings, texture=&AtmosphericCloudNoise01}
  n7 = ScalarParameter() {defaultValue=0.01, group=1_LightingDragon_ShadingDots, parameterName=ScreenSpace_Noise_Intensity}
  n8 = Add(
   A=@n5.EmissiveColor
   B=MaterialFunctionCall!(In=Multiply(A=@n6.R, B=@n7) {constB=0.02}, LUT, LUT Multiplier) {materialFunction=/Game/Packages/Fortress_SharedMaterials/Anime_Master_Material/MF_ApplyLUT}.Result
  )
  NamedRerouteDeclaration(None=SetMaterialAttributes(MaterialAttributes=@n5.MaterialAttributes, EmissiveColor=Saturate(None=@n8)) {attributeSetTypes [#1]}) {name=MA_Result, nodeColor=<rgba 1 0.7736 0 1>}
  n9 = VectorParameter() {group=1_LightingDragon_ShadingDots, parameterName=ShadingDots_Color, sortPriority=10, defaultValue=<rgba 0.2 0.2 0.2 1>}
  n10 = Multiply(A=TextureCoordinate()) {constB=120}
  n11 = FunctionInput() {bUsePreviewValueAsDefault=true, inputName=PreskinnedPosition}
  n12 = Multiply(A=Add(A=ComponentMask(None=@n11) {g=true}) {constB=-2}) {constB=0.15}
  n13 = Multiply(A=Distance(A=Frac(None=@n10), B=Constant() {r=0.5}), B=LinearInterpolate(Alpha=Saturate(None=@n12)) {constA=2, constB=4}) {constB=2}
  n14 = Power(Base=OneMinus(None=Saturate(None=@n13))) {constExponent=0.1}
  n15 = NamedRerouteUsage()
  n16 = Multiply(A=OneMinus(None=Power(Base=@n15) {constExponent=0.1})) {constB=10}
  NamedRerouteDeclaration(None=Multiply(A=Multiply(A=@n9.RGB, B=@n14), B=Saturate(None=@n16))) {name=ShadingDots, nodeColor=<rgba 1 0 0.5859 1>}
  n17 = MaterialFunctionCall() {materialFunction=&CameraDirectionVector}
  n18 = Multiply(A=Constant3Vector() {constant=<rgba 1 1 0 0>}, B=@n17.Result)
  n19 = Add(A=Multiply(B=Normalize(VectorInput=@n18)) {constA=0.25}, B=@n17.Result)
  n20 = Multiply(A=@n17.Result) {constB=50}
  n21 = Add(A=ActorPositionWS(), B=Constant3Vector() {constant=<rgba 0 0 0 1>})
  n22 = LinearInterpolate(
   A=Constant() {r=64}
   B=Constant() {r=128}
   Alpha=InverseLinearInterpolate(
    A=Constant() {r=45}
    B=Constant() {r=90}
    Value=MaterialFunctionCall() {materialFunction=/Engine/Functions/Engine_MaterialFunctions02/Utility/FOV}.FOV angle in degrees
   )
  )
  n23 = ShadingPathSwitch(Default=@n22, Mobile=Multiply(A=@n22) {constB=1.5})
  n24 = ScalarParameter() {defaultValue=1.5, parameterName=Dots_OverallScale}
  NamedRerouteDeclaration(None=MaterialFunctionCall(Plane Normal Axis=Normalize(VectorInput=@n19), Plane Center=Add(A=@n20, B=@n21), UV Scale=Multiply(A=@n23, B=@n24) {constB=2}) {materialFunction=&VirtualPlaneCoordinates}.UVs) {name=ObjectAlighedScreenPosition, nodeColor=<rgba 1 0.867 0 1>}
  n25 = DotProduct(A=Custom() {code="return View.DirectionalLightDirection;", description=DirectionalLightDirection, inputs []}, B=VertexNormalWS())
  n26 = Multiply(A=NamedRerouteDeclaration(None=Multiply(A=@n25) {constB=-1}) {name=LdotN, nodeColor=<rgba 0.2656 0 1 1>}) {constB=0.5}
  NamedRerouteDeclaration(None=Saturate(None=Add(A=@n26) {constB=0.5})) {name=Lambert, nodeColor=<rgba 0 0.9297 1 1>}
  n27 = Subtract(A=Distance(A=ActorPositionWS(), B=CameraPositionWS())) {constB=100}
  NamedRerouteDeclaration(None=Saturate(None=Divide(A=@n27) {constB=400})) {name=DistanceBlending, nodeColor=<rgba 0.4766 1 0 1>}
  NamedRerouteDeclaration(None=ComponentMask(None=FunctionInput() {inputName=FX, inputType=FunctionInput_Vector4, sortPriority=4}) {r=true}) {name=ShadingMask, nodeColor=<rgba 0 0.2031 1 1>}
  NamedRerouteDeclaration(None=FunctionInput() {inputName=MA, inputType=FunctionInput_MaterialAttributes}) {name=MA, nodeColor=<rgba 0.3594 1 0 1>}
  NamedRerouteDeclaration(None=ComponentMask(None=FunctionInput() {bUsePreviewValueAsDefault=true, inputName=SSC, sortPriority=3}) {r=true}) {name=Threshold, nodeColor=<rgba 0 0.6953 1 1>}
  NamedRerouteDeclaration(None=FunctionInput() {bUsePreviewValueAsDefault=true, inputName=InkLine Mask, inputType=FunctionInput_Scalar, sortPriority=2}) {name=DF_Line_Mask, nodeColor=<rgba 1 0.3281 0 1>}
 ]}
