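# Toon-shading material function for the HintLimpLatke shoe cosmetic skins, producing cloth-model material attributes with anime lighting baked into emissive.
# A selectable sun, screen-space or camera light vector drives a curve-atlas shading ramp with SSC threshold and backlit rim, blending tinted shaded and lit colors.
# Distance-field ink lines and SSC maps adapt by camera distance, with mobile and low-quality line variants.
# A shoe-specific ShadingDots function processes the final attributes.
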MaterialFunction /CosmeticShoes/Assets/FORT_Shoes/Shoe_HintLimpLatke/Skins/MF_AnimeShading_HintLimpLatke
graph {
 properties {libraryCategoriesText [Toon]}
 nodes [
  n1 = VectorParameter() {group=Toon_Shading, parameterName=ScreenSpaceLightVector, defaultValue=<rgba 1 1 -1 1>}
  Composite!(LightVector=@n1.RGB) {subgraphName=ProjectScreenAlighned}
  n2 = TextureSampleParameter2D(UVs=NamedRerouteUsage()) {
   group=Toon_Shading
   parameterName=LitDiffuse
   samplerSource=SSM_Wrap_WorldGroupSettings
   sortPriority=1
   texture=&T_WhitesRGB
  }
  n3 = FunctionInput(Preview=@n2.RGB) {bUsePreviewValueAsDefault=true, inputName=LitDiffuseOverride, sortPriority=1}
  n4 = Multiply(
   A=VectorParameter() {group=Toon_Shading, parameterName=TintColor, defaultValue=<rgba 0.054 0.054 0.054 1>}.RGB
   B=ScalarParameter() {defaultValue=1, group=Toon_Shading, parameterName=TintColor_Darkening, sliderMax=1}
  )
  n5 = LinearInterpolate(
   A=MaterialFunctionCall!(ShadedBaseColorIn=Multiply(A=@n3, B=@n4)) {materialFunction=/Game/Packages/Fortress_SharedMaterials/Cinematics/Anime/MF_CINE_ShadedDiffuseTint}.Result
   B=MaterialFunctionCall!(LitBaseColorIn=@n3) {materialFunction=/Game/Packages/Fortress_SharedMaterials/Cinematics/Anime/MF_CINE_LitDiffuseTint}.Result
   Alpha=NamedRerouteUsage()
  )
  n6 = Multiply(A=ScalarParameter() {defaultValue=0.5, group=Toon_Shading, parameterName=ThresholdDarkeningValue, sliderMax=1}, B=@n5)
  n7 = NamedRerouteUsage()
  n8 = MaterialFunctionCall!(In=LinearInterpolate(A=LinearInterpolate(A=@n6, B=@n5, Alpha=@n7), B=@n5, Alpha=NamedRerouteUsage())) {materialFunction=/Game/Packages/Fortress_SharedMaterials/Anime_Master_Material/MF_TODCustomColors}
  n9 = LinearInterpolate(
   A=NamedRerouteUsage()
   B=NamedRerouteDeclaration(None=MaterialFunctionCall!(In=@n8.Result, LUT, LUT Multiplier) {materialFunction=/Game/Packages/Fortress_SharedMaterials/Anime_Master_Material/MF_ApplyLUT}.Result) {name=Base Color OUT, nodeColor=<rgba 0 1 0.5705 1>}
   Alpha=NamedRerouteUsage()
  )
  n10 = ScalarParameter() {defaultValue=1, group=Toon_Shading, parameterName=GlobalBrightness, sliderMax=2}
  n11 = MaterialFunctionCall!(
   Emissive In=MaterialFunctionCall!(Color=Multiply(A=@n9, B=@n10)) {materialFunction=/Game/Packages/Fortress_SharedMaterials/Anime_Master_Material/MF_TODColorAdjustment}.Result
   LightingExtrasColorData=NamedRerouteUsage()
   CINE_LightSources=NamedRerouteUsage()
   ShadingCurve=NamedRerouteUsage()
  ) {materialFunction=/Game/Packages/Fortress_SharedMaterials/Cinematics/Anime/MF_CINE_FinalColorBlend}
  n12 = ChannelMaskParameter(None=NamedRerouteUsage()) {group=Toon_Shading_SSC, maskChannel=Green, parameterName=SSC_OverlapLineValueChannel, sortPriority=10, defaultValue=<rgba 0 1 0 0>}
  n13 = ScalarParameter() {defaultValue=0.001, group=Toon_Shading, parameterName=Subsurface Mask Value, sliderMax=1}
  n14 = SetMaterialAttributes!(
   MaterialAttributes=NamedRerouteUsage()
   BaseColor=Constant3Vector() {desc="dont touch!"}
   Metallic=ComponentMask(None=NamedRerouteUsage()) {desc=AO / Threshold, r=true}
   Specular=Constant() {desc="dont touch!"}
   EmissiveColor=Add(A=NamedRerouteUsage(), B=NamedRerouteDeclaration(None=@n11.Result) {name=FinalColor, nodeColor=<rgba 1 0.4967 0 1>})
   Roughness=LinearInterpolate(A=@n12.R, B=OneMinus(None=@n12.R), Alpha=ScalarParameter() {group=Toon_Shading_SSC, parameterName=Invert Roughness Pass, sliderMax=1})
   AmbientOcclusion=NamedRerouteUsage()
   SubsurfaceColor=MaterialFunctionCall(Y=@n13, Z=@n13) {materialFunction=&MakeFloat3}.Result
   CustomData0=Constant() {desc=Opacity, r=1}
   ShadingModel=ShadingModel() {shadingModel=MSM_Cloth}
  ) {attributeSetTypes [
    #1
    #2
    #3
    #4
    #5
    #6
    #7
    #8
    #9
   ]}
  FunctionOutput(None=MaterialFunctionCall!(
    MA=@n14
    PreskinnedPosition=FunctionInput() {bUsePreviewValueAsDefault=true, inputName=PreskinnedPosition}
    InkLine Mask=NamedRerouteUsage()
    SSC=NamedRerouteUsage()
    FX=NamedRerouteUsage()
   ) {materialFunction=&/MF_F_MED_HintLimpLatke_ShadingDots}.Result) {sortPriority=10}
  n15 = TextureSampleParameter2D(UVs=NamedRerouteUsage()) {
   group=Toon_Shading
   parameterName=SSC_Texture
   samplerSource=SSM_Wrap_WorldGroupSettings
   samplerType=SAMPLERTYPE_Masks
   sortPriority=1
   texture=&T_BlackMask
  }
  n16 = MaterialFunctionCall(Float4=@n15.RGBA) {materialFunction=&BreakOutFloat4Components}
  n17 = StaticSwitchParameter(
   True=MaterialFunctionCall(
    X=@n16.R
    Y=@n16.R
    Z=@n16.R
    A=FunctionInput() {bUsePreviewValueAsDefault=true, inputName=SSC_AlphaOverride, inputType=FunctionInput_Scalar, sortPriority=1}
   ) {materialFunction=&MakeFloat4}.Result
   False=@n15.RGBA
  ) {group=Toon_Shading_SSC, parameterName=SSC_OverrideAlphaChannel, sortPriority=0}
  n18 = TextureSampleParameter2D(UVs=NamedRerouteUsage()) {
   group=Toon_Shading
   parameterName=SSC_Texture_Far
   samplerSource=SSM_Wrap_WorldGroupSettings
   samplerType=SAMPLERTYPE_Masks
   sortPriority=1
   texture=&T_BlackMask
  }
  n19 = ScalarParameter() {defaultValue=100, group=Toon_Shading_SSC, parameterName=SSC_FarDistance, sliderMax=1000}
  n20 = Subtract(A=Distance(A=WorldPosition().XYZ, B=CameraPositionWS()), B=@n19)
  n21 = Subtract(A=ScalarParameter() {defaultValue=500, group=Toon_Shading_SSC, parameterName=SSC_NearDistance, sliderMax=1000}, B=@n19)
  n22 = Divide(
   A=ComponentMask(None=ViewProperty() {property=MEVP_BufferSize}.Property) {r=true}
   B=ScalarParameter() {defaultValue=3840, group=Toon_Shading_SSC, parameterName=Reference Screen Size, sliderMax=3840}
  )
  n23 = Saturate(None=Divide(A=Divide(A=@n20, B=@n21), B=Power(Base=@n22) {constExponent=1.2}))
  NamedRerouteDeclaration(None=StaticSwitchParameter(True=LinearInterpolate(A=@n17, B=@n18.RGBA, Alpha=@n23), False=@n17) {group=Toon_Shading, parameterName=Enable Far SSC, sortPriority=0}) {name=SSC Map, nodeColor=<rgba 1 0 0.4453 1>}
  n24 = TextureSampleParameter2D(UVs=NamedRerouteUsage()) {
   group=Skin FX
   parameterName=SkinFX_Mask
   samplerSource=SSM_Wrap_WorldGroupSettings
   samplerType=SAMPLERTYPE_Masks
   sortPriority=0
   texture=&T_BlackMask
  }
  NamedRerouteDeclaration(None=@n24.RGBA) {name=SkinFX, nodeColor=<rgba 1 0.3984 0 1>}
  n25 = GetMaterialAttributes(None=FunctionInput() {inputName=Material, inputType=FunctionInput_MaterialAttributes}) {attributeGetTypes [#4]}
  NamedRerouteDeclaration(None=@n25.EmissiveColor) {name=EmissiveColor, nodeColor=<rgba 1 1 0 1>}
  n26 = Constant3Vector() {constant=<rgba 0 0 1 0>}
  n27 = ChannelMaskParameter(None=NamedRerouteUsage()) {
   desc="Select in which channel of the FX map the mask camera light vector will be stored "
   group=Toon_Shading
   parameterName=CameraLightVector_FXChannel
   sortPriority=10
  }
  MaterialExpressionReroute_12 = Reroute(None=@n1.RGB)
  n28 = Normalize(VectorInput=@MaterialExpressionReroute_12.LightVector)
  n29 = ComponentMask(None=@n28) {r=true}
  n30 = Constant3Vector() {constant=<rgba 0 1 0 0>}
  n31 = Normalize(VectorInput=WorldPosition() {worldPositionShaderOffset=WPT_CameraRelative}.XYZ)
  n32 = Normalize(VectorInput=CrossProduct(A=Transform(None=@n30) {transformSourceType=TRANSFORMSOURCE_View}, B=@n31))
  n33 = ComponentMask(None=@n28) {g=true}
  n34 = CrossProduct(A=@n31, B=@n32)
  n35 = Add(A=Add(A=Multiply(A=@n29, B=@n32), B=Multiply(A=@n33, B=@n34)), B=Multiply(A=ComponentMask(None=@n28) {b=true}, B=@n31))
  MaterialExpressionReroute_14 = Reroute(None=@n35)
  n36 = ComponentMask(None=Custom() {code="return View.DirectionalLightColor;", description=DirectionalLightColor, outputType=CMOT_Float4, inputs []}) {b=true, g=true, r=true}
  n37 = Custom() {code="return View.DirectionalLightDirection;", description=DirectionalLightDirection, inputs []}
  n38 = CollectionParameter() {
   collection=&FortniteMaterialParameters
   parameterName=SunlightDirectionalVector
  }
  n39 = ComponentMask(None=@n38) {b=true, g=true, r=true}
  n40 = If(A=DotProduct(A=@n37, B=@n37), "A > B"=@n37, "A == B"=@n39, "A < B"=@n39) {desc=check if DirectionalLight exists}
  n41 = StaticSwitchParameter(
   True=@MaterialExpressionReroute_14.Result
   False=MaterialFunctionCall!(In=If(A=@n36, "A > B"=@n40, "A == B"=@n39, "A < B"=@n39) {desc=check if DirectionalLight color is Black}) {materialFunction=/Game/Packages/Fortress_SharedMaterials/Cinematics/Anime/MF_CINE_AnimeLightDirOverride}.Result
  ) {
   desc="Use a Screen / View alight light vector. \r\nOverrides the UseDirectionalLightFunction!"
   group=Toon_Shading
   parameterName=UseScreenSpaceLightDirection
   sortPriority=0
  }
  n42 = MaterialFunctionCall(Float3=@n41) {materialFunction=&BreakOutFloat3Components}
  n43 = ScalarParameter() {
   desc="0.0 = the light vector is locked on the XY plane.\r\n1.0 = the actual light vector is used  \r\n(for Faces use a blend value of 0,0)"
   group=Toon_Shading
   parameterName=LockLightZ_Blend
   sliderMax=1
  }
  n44 = MaterialFunctionCall(X=@n42.R, Y=@n42.R, Z=Multiply(A=@n42.B, B=@n43)) {materialFunction=&MakeFloat3}
  n45 = StaticSwitchParameter(True=LinearInterpolate(A=@n41, B=Normalize(VectorInput=@n44.Result), Alpha=VertexColor().A), False=@n41) {
   desc="Lock the Z axis of the Light Vector\r\nis driven by the VertexColor A channel (0 = Regular LightVector, 1= Locked LIghtVector)"
   group=Toon_Shading
   parameterName=UseLockLightZVector
   sortPriority=0
  }
  n46 = ComponentMask(None=@n45) {b=true, g=true, r=true}
  n47 = StaticSwitchParameter(
   True=LinearInterpolate(A=@n46, B=Transform(None=@n26) {desc=Camera Forward Vector, transformSourceType=TRANSFORMSOURCE_Camera}, Alpha=@n27.R)
   False=@n46
  ) {
   desc="Use the Camera vector instead of the light vector for an reaz which is defined by the FX map. Select the FX channel with this parameter \"CameraLightVector_FXChannel\"."
   group=Toon_Shading
   parameterName=UseCameraLightVector
   sortPriority=0
  }
  NamedRerouteDeclaration(None=@n47) {name=LightVector, nodeColor=<rgba 0 1 0.8634 1>}
  NamedRerouteDeclaration(None=@n25.MaterialAttributes) {name=MA, nodeColor=<rgba 1 1 0 1>}
  n48 = NamedRerouteUsage()
  n49 = TextureSampleParameter2D(UVs=@n48) {
   group=Toon_Shading
   parameterName=DistanceField_InkLines
   samplerSource=SSM_Wrap_WorldGroupSettings
   samplerType=SAMPLERTYPE_Alpha
   sortPriority=1
   texture=/Game/Global/Textures/Default/Blanks/T_BlackAlpha
  }
  n50 = TextureSampleParameter2D(UVs=@n48) {
   group=Toon_Shading
   parameterName=DistanceField_InkLines_Mobile
   samplerSource=SSM_Wrap_WorldGroupSettings
   samplerType=SAMPLERTYPE_Alpha
   sortPriority=1
   texture=&T_BlackAlpha
  }
  n51 = StaticSwitchParameter(
   True=QualitySwitch(Default=@n49.RGBA, Low=@n50.RGBA, High=@n49.RGBA, Medium=@n50.RGBA, Epic=@n49.RGBA)
   False=StaticSwitchParameter(
    True=FeatureLevelSwitch(Default=@n49.RGBA, ES2=@n50.RGBA, ES3_1=@n50.RGBA, SM4_REMOVED=@n49.RGBA, SM5=@n49.RGBA, SM6=@n49.RGBA)
    False=@n49.RGBA
   ) {group=Toon_Shading, parameterName=Enable Unique Mobile DF Inklines, sortPriority=0}
  ) {group=Toon_Shading, parameterName=Enable Unique Low Quality DF Inklines, sortPriority=0}
  n52 = MaterialFunctionCall!(Roughness input=StaticSwitchParameter(True=Power(Base=@n51) {constExponent=0.4545}, False=@n51) {group=Toon_Shading, parameterName=UseGammaCorrectedDFLines, sortPriority=0}) {materialFunction=/Game/Packages/Fortress_SharedMaterials/Anime_Master_Material/MF_OrientationFade}
  n53 = FunctionInput() {
   bUsePreviewValueAsDefault=true
   desc=This mask only affects DistanceField_InkLines texture parameter
   inputName=DF InkLine Visibility Mask
   inputType=FunctionInput_Scalar
   sortPriority=10
   previewValue=<1 0 0 1>
  }
  n54 = Max(
   A=FunctionInput() {bUsePreviewValueAsDefault=true, inputName=DF Inkline Injection, inputType=FunctionInput_Scalar, sortPriority=11}
   B=Multiply(A=ChannelMaskParameter(None=@n52.Result) {group=Toon_Shading_DF_InkLines, parameterName=DistanceField_Channel, sortPriority=10}.R, B=@n53) {desc=Mask Inkline Here}
  ) {desc=Inject Inklines here}
  n55 = ScalarParameter() {defaultValue=4, group=Toon_Shading_DF_InkLines, parameterName=DF_BlurPow, sliderMax=16}
  n56 = OneMinus(None=Multiply(
    A=Power(Base=@n54, Exp=@n55) {constExponent=3}
    B=ScalarParameter() {defaultValue=1, group=Toon_Shading_DF_InkLines, parameterName=DF_BlurMult, sliderMax=2}
   ) {constB=0.5})
  n57 = Constant() {r=1}
  n58 = ScalarParameter() {defaultValue=0.4, group=Toon_Shading_DF_InkLines, parameterName=DF_LineThickness_Near, sliderMax=1}
  n59 = ScalarParameter() {defaultValue=0.4, group=Toon_Shading_DF_InkLines, parameterName=DF_LineThickness_Near_Mobile, sliderMax=1}
  n60 = ScalarParameter() {defaultValue=0.6, group=Toon_Shading_DF_InkLines, parameterName=DF_LineThickness_Far, sliderMax=1}
  n61 = ScalarParameter() {defaultValue=0.6, group=Toon_Shading_DF_InkLines, parameterName=DF_LineThickness_Far_Mobile, sliderMax=1}
  n62 = MaterialFunctionCall!(
   Fade Offset=ScalarParameter() {defaultValue=100, group=Toon_Shading_DF_InkLines, parameterName=DF_FarDistance, sliderMax=1000}
   Fade Distance=ScalarParameter() {defaultValue=500, group=Toon_Shading_DF_InkLines, parameterName=DF_NearDistance, sliderMax=1000}
  ) {materialFunction=/Game/Global/Materials/Character_Parent_2023/Anime_Parent_2023/MaterialFunctions/MFC_DistanceFade_FOV_compensation}
  n63 = StaticSwitchParameter(
   True=LinearInterpolate(A=ShadingPathSwitch(Default=@n58, Mobile=@n59), B=ShadingPathSwitch(Default=@n60, Mobile=@n61), Alpha=@n62.Result)
   False=LinearInterpolate(A=@n58, B=@n60, Alpha=@n62.Result)
  ) {group=Toon_Shading, parameterName=Enable Mobile DF Inklines Thickness, sortPriority=0}
  n64 = ScalarParameter() {defaultValue=0.8, group=Toon_Shading_DF_InkLines, parameterName=DF_LineHardness, sliderMax=1}
  n65 = OneMinus(None=Saturate(None=SphereMask(A=@n54, B=@n57, Radius=@n63, Hardness=@n64) {attenuationRadius=0.1}))
  NamedRerouteDeclaration(None=Saturate(None=Multiply(A=@n56, B=@n65))) {name=InkLine Mask, nodeColor=<rgba 0.1615 0.1615 0.1615 1>}
  n66 = TextureSampleParameter2D(UVs=NamedRerouteUsage()) {
   group=Toon_Shading
   parameterName=InkLineColor_Texture
   samplerSource=SSM_Wrap_WorldGroupSettings
   sortPriority=1
   texture=&T_BlacksRGB
  }
  NamedRerouteDeclaration(None=MaterialFunctionCall(In=@n66.RGB) {materialFunction=&MF_ApplyLUT}.Result) {name=InkLine Color, nodeColor=<rgba 0.1615 0.1615 0.1615 1>}
  NamedRerouteDeclaration(None=FunctionInput(Preview=TextureCoordinate()) {bUsePreviewValueAsDefault=true, inputName=CustomUVs, inputType=FunctionInput_Vector2, sortPriority=15}) {name=CustomUVs, nodeColor=<rgba 1 0.0993 0 1>}
  n67 = NamedRerouteDeclaration(None=Normalize(VectorInput=PixelNormalWS())) {name=Adjusted Normals, nodeColor=<rgba 0.5 0.5 1 1>}
  n68 = MaterialFunctionCall!(In=DotProduct(A=NamedRerouteUsage(), B=@n67)) {materialFunction=/Game/Packages/Fortress_SharedMaterials/Cinematics/Anime/MF_CINE_AnimeLightingFeatures}
  NamedRerouteDeclaration(None=@n68.CINE_LightSources) {name=CINE_LightSources, nodeColor=<rgba 1 0.9609 0 1>}
  NamedRerouteDeclaration(None=@n68.CINE_LightColor) {name=CINE_LightColor, nodeColor=<rgba 1 0.6803 0 1>}
  n69 = NamedRerouteUsage()
  n70 = MaterialFunctionCall(
   Alpha=Power(
    Base=ChannelMaskParameter(None=@n69) {group=Toon_Shading_SSC, parameterName=SSC_ThresholdChannel, sortPriority=10}.R
    Exp=ScalarParameter() {defaultValue=2, group=Toon_Shading, parameterName=ThresholdmaskPower, sliderMax=8}
   )
   Min=ScalarParameter() {group=Toon_Shading, parameterName=Shaded_Threshold_Min, sliderMax=1}
   Max=ScalarParameter() {defaultValue=1, group=Toon_Shading, parameterName=Shaded_Threshold_Max, sliderMax=1}
  ) {materialFunction=&SmoothStep}
  NamedRerouteDeclaration(None=Round(None=@n70.Result)) {name=SSC Throshold, nodeColor=<rgba 0.7973 1 0 1>}
  n71 = Saturate(None=@n68.CINE_ShadingCurve)
  NamedRerouteDeclaration(None=Round(None=@n71)) {name=Shading Dot, nodeColor=<rgba 0.7973 1 0 1>}
  n72 = Add(A=@n68.CINE_ShadingCurve, B=ScalarParameter() {group=Toon_Shading, parameterName=Lit Bias, sliderMax=1})
  n73 = ScalarParameter() {defaultValue=12, group=Toon_Shading_RimHighlighBacklit, parameterName=BackLitRim_OutlineC2, sliderMax=16, sliderMin=-8}
  n74 = Fresnel(
   ExponentIn=ScalarParameter() {defaultValue=2, group=Toon_Shading_RimHighlighBacklit, parameterName=BackLitExponent1, sliderMax=8}
   Normal=@n67
  ) {baseReflectFraction=0}
  n75 = LinearInterpolate(
   A=ScalarParameter() {defaultValue=-4, group=Toon_Shading_RimHighlighBacklit, parameterName=BackLitRim_OutlineC1, sliderMax=16}
   B=@n73
   Alpha=Power(
    Base=OneMinus(None=@n74)
    Exp=ScalarParameter() {defaultValue=4, group=Toon_Shading_RimHighlighBacklit, parameterName=BacklitOutlineExponent2, sliderMax=8}
   )
  )
  n76 = Add(A=OneMinus(None=Saturate(None=@n75)), B=@n71)
  n77 = Add(
   A=DotProduct(
    A=MaterialFunctionCall() {materialFunction=&CameraDirectionVector}.Result
    B=NamedRerouteUsage()
   )
   B=ScalarParameter() {defaultValue=0.25, group=Toon_Shading_RimHighlighBacklit, parameterName=BackLitBias, sliderMax=1}
  )
  n78 = ScalarParameter() {defaultValue=0.75, group=Toon_Shading_RimHighlighBacklit, parameterName=BackLitMaxBias, sliderMax=1}
  n79 = CurveAtlasRowParameter(CurveTime=LinearInterpolate(A=Multiply(A=@n72, B=@n70.Result), B=Multiply(A=@n76, B=@n70.Result), Alpha=Clamp(None=@n77, Max=@n78))) {
   atlas=/Game/Characters/Shared/CurveAtlas/AthenaCharacterCurveAtlas
   curve=/Game/Packages/Fortress_SharedMaterials/Anime_Master_Material/TwoToneShadingCurve
   defaultValue=14
   group=Toon_Shading
   parameterName=ShadingCurve
   sortPriority=5
  }
  NamedRerouteDeclaration(None=ChannelMaskParameter(None=@n79) {group=Toon_Shading, parameterName=ShadingCurveChannel, sortPriority=10}.R) {name=Shading Curve, nodeColor=<rgba 0.7973 1 0 1>}
  PinBase() {reroutePins [LightVector {expression=&self:MaterialExpressionReroute_12}]}
  PinBase!(Result=@n35) {reroutePins [Result {expression=&self:MaterialExpressionReroute_14}]}
 ]
}
