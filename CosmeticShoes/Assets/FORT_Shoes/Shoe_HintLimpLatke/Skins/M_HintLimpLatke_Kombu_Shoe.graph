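# Cloth-shaded material for the Kombu skin of the HintLimpLatke cosmetic shoe, for skeletal meshes with morph targets and mesh deformer support.
# It chains material functions through material attributes: the 2023 character parent PreFX feeds HintLimpLatke anime shading, then the shared Kicks sock function, then the parent PostFX.
# PreFX outputs also go directly into PostFX, and the graph adds no logic of its own.
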Material /CosmeticShoes/Assets/FORT_Shoes/Shoe_HintLimpLatke/Skins/M_HintLimpLatke_Kombu_Shoe
graph {
 properties {
  bAutomaticallySetUsageInEditor=false
  bUseMaterialAttributes=true
  bUsedWithClothing=true
  bUsedWithMeshDeformer=true
  bUsedWithMorphTargets=true
  bUsedWithSkeletalMesh=true
  pixelDepthOffsetMode=PDOM_Legacy
  shadingModel=MSM_Cloth
  usedShadingModels=Cloth
 }
 nodes [
  n1 = MaterialFunctionCall() {materialFunction=&MF_CharacterParent_2023_PreFX}
  MaterialFunctionCall(
   Material=MaterialFunctionCall!(
    InOpacity
    Material=MaterialFunctionCall!(
     PreskinnedPosition=@n1.Pre skinned Local Position
     Material=@n1.Pre skinned Local Position
     LitDiffuseOverride
     SSC_AlphaOverride
     DF InkLine Visibility Mask
     DF Inkline Injection
     CustomUVs
    ) {materialFunction=&/MF_AnimeShading_HintLimpLatke}.Result
    Mask
   ) {materialFunction=/Game/Global/Materials/Cosmetics/Kicks/Socks/Materials/MF_Sock}.Result
   M=@n1.M
   Pre Skinned Local Position=@n1.M
   Distance Blend=@n1.M
   Fresnel=@n1.M
  ) {materialFunction=&MF_CharacterParent_2023_PostFX}
 ]
}
